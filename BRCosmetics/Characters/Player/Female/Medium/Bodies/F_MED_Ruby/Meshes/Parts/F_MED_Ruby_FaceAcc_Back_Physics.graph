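# Physics asset for the back face-accessory part of the medium female Ruby character, setting up capsule bodies and joints for her hair.
# Left, middle and right four-bone chains plus a short left-shoulder strand hang from head-linked bases, with tight root limits and wider swing below.
# Head, neck, upper spine, clavicles and left upper arm add body capsules joined by angularly free constraints.
# Similar free constraints cross-link both side-chain tips to the middle tip.
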PhysicsAsset /BRCosmetics/Characters/Player/Female/Medium/Bodies/F_MED_Ruby/Meshes/Parts/F_MED_Ruby_FaceAcc_Back_Physics
bodies {
 base_lshoulderhair [{shapeType=Capsule, length=0.505, radius=0.505, shapeName=base_lshoulderhair_capsule, rotation=<rot 0 -90 0>}]
 clavicle_l [{shapeType=Capsule, length=11, radius=5.3, shapeName=clavicle_l_capsule, center=<7 -1.2 -1>, rotation=<rot 75 -0.0023 -0.0024>}]
 clavicle_r [{shapeType=Capsule, length=6, radius=5.8, shapeName=clavicle_l_capsule, center=<-7 1.2 1>, rotation=<rot 72 0 0>}]
 dyn_hair_l_01 [{shapeType=Capsule, length=2.5, radius=2.5102, shapeName=dyn_hair_l_01_capsule, center=<2.571 -1.5 1.5>, rotation=<rot 0 90 90>}]
 dyn_hair_l_02 [{shapeType=Capsule, length=2, radius=2, shapeName=dyn_hair_l_02_capsule, center=<2.5588 0 0>, rotation=<rot 0 90 90>}]
 dyn_hair_l_03 [{shapeType=Capsule, length=3.7129, radius=2.543, shapeName=dyn_hair_l_03_capsule, center=<2.7522 0 0>, rotation=<rot 0 90 90>}]
 dyn_hair_l_04 [{shapeType=Capsule, length=11, radius=1.6, shapeName=dyn_hair_l_04_capsule, center=<2.6854 -2.25 0>, rotation=<rot 80 3 20>}]
 dyn_hair_l_base [{shapeType=Capsule, length=0.505, radius=0.505, shapeName=dyn_hair_l_base_capsule, rotation=<rot 0 -90 0>}]
 dyn_hair_mid_01 [{shapeType=Capsule, length=2.5, radius=2.5, shapeName=dyn_hair_l_01_capsule, center=<2.571 -1.3 0>, rotation=<rot 0 90 90>}]
 dyn_hair_mid_02 [{shapeType=Capsule, length=6, radius=2.8, shapeName=dyn_hair_mid_02_capsule, center=<1.8883 0.5 -2.8>, rotation=<rot 0 0 -15>}]
 dyn_hair_mid_03 [{shapeType=Capsule, length=6, radius=2.8, shapeName=dyn_hair_mid_03_capsule, center=<2.2995 0.5 -2.6>, rotation=<rot 0 0 -10>}]
 dyn_hair_mid_04 [{shapeType=Capsule, length=7, radius=2, shapeName=dyn_hair_mid_04_capsule, center=<3.2654 -0.3 -0.6>, rotation=<rot 0 80 90>}]
 dyn_hair_mid_base [{shapeType=Capsule, length=0.505, radius=0.505, shapeName=dyn_hair_mid_base_capsule, rotation=<rot 0 -90 0>}]
 dyn_hair_r_01 [{shapeType=Capsule, length=2.5, radius=2.5102, shapeName=dyn_hair_l_01_capsule, center=<2.571 -0.8 -0.7>, rotation=<rot 0 90 90>}]
 dyn_hair_r_02 [{shapeType=Capsule, length=7, radius=2.5, shapeName=dyn_hair_r_02_capsule, center=<1.7245 3.1 -0.1>, rotation=<rot 0 85 80>}]
 dyn_hair_r_03 [{shapeType=Capsule, length=5, radius=3.1, shapeName=dyn_hair_r_03_capsule, center=<1.967 -0.25 -1.1>, rotation=<rot 0 60 90>}]
 dyn_hair_r_04 [{shapeType=Capsule, length=9, radius=1.7, shapeName=dyn_hair_r_04_capsule, center=<3.4 -2 1>, rotation=<rot -45 0.0002 -10>}]
 dyn_hair_r_base [{shapeType=Capsule, length=0.505, radius=0.505, shapeName=dyn_hair_r_base_capsule, rotation=<rot 0 -90 0>}]
 dyn_lshoulderhair [{shapeType=Capsule, length=4, radius=2.8, shapeName=dyn_lshoulderhair_capsule, center=<1.2169 0.9 -0.8>, rotation=<rot 90 165.9638 -98.0725>}]
 head [{shapeType=Capsule, length=6, radius=6, shapeName=head_capsule, center=<3.5 1.1987 0>, rotation=<rot 90 0 0>}]
 neck_01 [{shapeType=Capsule, length=8, radius=4, shapeName=neck_01_capsule, center=<3.6683 1 0>, rotation=<rot 90 0 0>}]
 neck_02 [{shapeType=Capsule, length=33, radius=4.5, shapeName=spine_05_capsule, center=<-1.4307 6.0024 0>}]
 spine_05 [{shapeType=Capsule, length=9, radius=8.5, shapeName=spine_05_capsule, center=<2.691 -0.3267 0.5034>, rotation=<rot -5 0 -5>}]
 upperarm_l [{shapeType=Capsule, length=13.005, radius=5.5, shapeName=upperarm_l_capsule, center=<10 0 -1>, rotation=<rot 90 0 0>}]
}
constraints [
 {
  bone1Name=neck_01
  bone2Name=spine_05
  swing1LimitDegrees=45
  swing1Motion=Free
  swing2LimitDegrees=45
  swing2Motion=Free
  twistLimitDegrees=45
  twistMotion=Free
 }
 {
  bone1Name=head
  bone2Name=neck_02
  swing1LimitDegrees=45
  swing1Motion=Free
  swing2LimitDegrees=45
  swing2Motion=Free
  twistLimitDegrees=45
  twistMotion=Free
 }
 {
  bone1Name=clavicle_l
  bone2Name=spine_05
  swing1LimitDegrees=45
  swing1Motion=Free
  swing2LimitDegrees=45
  swing2Motion=Free
  twistLimitDegrees=45
  twistMotion=Free
 }
 {
  bone1Name=clavicle_r
  bone2Name=spine_05
  swing1LimitDegrees=45
  swing1Motion=Free
  swing2LimitDegrees=45
  swing2Motion=Free
  twistLimitDegrees=45
  twistMotion=Free
 }
 {
  bone1Name=dyn_hair_l_base
  bone2Name=head
  swing1LimitDegrees=45
  swing1Motion=Free
  swing2LimitDegrees=45
  swing2Motion=Free
  twistLimitDegrees=45
  twistMotion=Free
 }
 {bone1Name=dyn_hair_l_01, bone2Name=dyn_hair_l_base, swing1LimitDegrees=5, swing2LimitDegrees=5, twistLimitDegrees=5, twistMotion=Locked}
 {bone1Name=dyn_hair_l_02, bone2Name=dyn_hair_l_01, swing1LimitDegrees=50, swing2LimitDegrees=15, twistLimitDegrees=5}
 {bone1Name=dyn_hair_l_03, bone2Name=dyn_hair_l_02, swing1LimitDegrees=50, swing2LimitDegrees=15, twistLimitDegrees=5}
 {bone1Name=dyn_hair_l_04, bone2Name=dyn_hair_l_03, swing1LimitDegrees=50, swing2LimitDegrees=15, twistLimitDegrees=5}
 {
  bone1Name=dyn_hair_mid_base
  bone2Name=head
  swing1LimitDegrees=45
  swing1Motion=Free
  swing2LimitDegrees=45
  swing2Motion=Free
  twistLimitDegrees=45
  twistMotion=Free
 }
 {bone1Name=dyn_hair_mid_01, bone2Name=dyn_hair_mid_base, swing1LimitDegrees=5, swing2LimitDegrees=5, twistLimitDegrees=5}
 {bone1Name=dyn_hair_mid_02, bone2Name=dyn_hair_mid_01, swing1LimitDegrees=50, swing2LimitDegrees=15, twistLimitDegrees=5}
 {bone1Name=dyn_hair_mid_03, bone2Name=dyn_hair_mid_02, swing1LimitDegrees=50, swing2LimitDegrees=15, twistLimitDegrees=5}
 {bone1Name=dyn_hair_mid_04, bone2Name=dyn_hair_mid_03, swing1LimitDegrees=50, swing2LimitDegrees=15, twistLimitDegrees=5}
 {
  bone1Name=dyn_hair_r_base
  bone2Name=head
  swing1LimitDegrees=45
  swing1Motion=Free
  swing2LimitDegrees=45
  swing2Motion=Free
  twistLimitDegrees=45
  twistMotion=Free
 }
 {bone1Name=dyn_hair_r_01, bone2Name=dyn_hair_r_base, swing1LimitDegrees=5, swing2LimitDegrees=5, twistLimitDegrees=5, twistMotion=Locked}
 {bone1Name=dyn_hair_r_02, bone2Name=dyn_hair_r_01, swing1LimitDegrees=50, swing2LimitDegrees=15, twistLimitDegrees=5}
 {bone1Name=dyn_hair_r_03, bone2Name=dyn_hair_r_02, swing1LimitDegrees=50, swing2LimitDegrees=15, twistLimitDegrees=5}
 {bone1Name=dyn_hair_r_04, bone2Name=dyn_hair_r_03, swing1LimitDegrees=50, swing2LimitDegrees=15, twistLimitDegrees=5}
 {
  bone1Name=upperarm_l
  bone2Name=clavicle_l
  swing1LimitDegrees=45
  swing1Motion=Free
  swing2LimitDegrees=45
  swing2Motion=Free
  twistLimitDegrees=45
  twistMotion=Free
 }
 {bone1Name=dyn_lshoulderhair, bone2Name=base_lshoulderhair, swing1LimitDegrees=30, swing2LimitDegrees=30, twistLimitDegrees=30}
 {
  bone1Name=base_lshoulderhair
  bone2Name=head
  swing1LimitDegrees=45
  swing1Motion=Free
  swing2LimitDegrees=45
  swing2Motion=Free
  twistLimitDegrees=45
  twistMotion=Free
 }
 {
  bone1Name=dyn_hair_l_04
  bone2Name=dyn_hair_mid_04
  swing1LimitDegrees=45
  swing1Motion=Free
  swing2LimitDegrees=45
  swing2Motion=Free
  twistLimitDegrees=45
  twistMotion=Free
 }
 {
  bone1Name=dyn_hair_r_04
  bone2Name=dyn_hair_mid_04
  swing1LimitDegrees=45
  swing1Motion=Free
  swing2LimitDegrees=45
  swing2Motion=Free
  twistLimitDegrees=45
  twistMotion=Free
 }
 {
  bone1Name=neck_02
  bone2Name=neck_01
  swing1LimitDegrees=45
  swing1Motion=Free
  swing2LimitDegrees=45
  swing2Motion=Free
  twistLimitDegrees=45
  twistMotion=Free
 }
]
