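# Physics asset for the face accessory part of the female Ruby player cosmetic, providing capsule bodies and joints for its dynamic hair.
# A hair chain hangs from the head through a base bone and six segments, then branches into three tips, with angular limits tightest at the tips.
# Capsules on the head, neck, upper spine, both clavicles and only the left upper arm form the surrounding body, linked by free-rotation constraints.
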PhysicsAsset /BRCosmetics/Characters/Player/Female/Medium/Bodies/F_MED_Ruby/Meshes/Parts/F_MED_Ruby_FaceAcc_Physics
bodies {
 clavicle_l [{shapeType=Capsule, length=9, radius=5.3, shapeName=clavicle_l_capsule, center=<3.5 -0.2549 -0.3101>, rotation=<rot 75 -0.0023 -0.0024>}]
 clavicle_r [{shapeType=Capsule, length=9, radius=5.3, shapeName=clavicle_l_capsule, center=<-3.5 0.2549 0.3101>, rotation=<rot 75 -0.0023 179.9976>}]
 dyn_hair_01 [{shapeType=Capsule, length=11, radius=2, center=<6.5 0 1>, rotation=<rot 0 90 90>}]
 dyn_hair_02 [{shapeType=Capsule, length=2.1476, radius=2, center=<3 0 0>, rotation=<rot 0 90 90>}]
 dyn_hair_03 [{shapeType=Capsule, length=1.5, radius=3.3, center=<2.8612 -0.188 0>, rotation=<rot 0 115 90>}]
 dyn_hair_04 [{shapeType=Capsule, length=2, radius=3.3, center=<3 1.0687 0.5342>, rotation=<rot 20 90 0>}]
 dyn_hair_05 [{shapeType=Capsule, length=2.5, radius=3.3, center=<2.5926 1.0924 0>, rotation=<rot 0 90 0>}]
 dyn_hair_06 [{shapeType=Capsule, length=1, radius=4, center=<4.5 3.2 0>, rotation=<rot 0 90 90>}]
 dyn_hair_07 [{shapeType=Capsule, length=5.5, radius=1.3, center=<2.5 0.6 -0.4>, rotation=<rot 30 90 40>}]
 dyn_hair_08 [{shapeType=Capsule, length=2.3, radius=1.3, center=<1.1 0.26 0.25>, rotation=<rot 0 130 130>}]
 dyn_hair_09 [{shapeType=Capsule, length=1.8, radius=0.8, center=<1.7 0.2 0>, rotation=<rot 0 80 90>}]
 dyn_hair_base [{shapeType=Capsule, length=0.505, radius=0.505, shapeName=dyn_hair_base_capsule, rotation=<rot 0 -90 0>}]
 head [{shapeType=Capsule, length=8, radius=6, shapeName=head_capsule, center=<3.5 1.1987 0>, rotation=<rot 90 0 0>}]
 neck_01 [{shapeType=Capsule, length=8, radius=4, shapeName=neck_01_capsule, center=<3.6683 1 0>, rotation=<rot 90 0 0>}]
 neck_02 [{shapeType=Capsule, length=88, radius=3.5, shapeName=neck_02_capsule, center=<0 -2.3 0>, rotation=<rot 0 -90 0>}]
 spine_05 [{shapeType=Capsule, length=8.5598, radius=9, shapeName=spine_05_capsule, center=<-4.1075 2.6915 0>}]
 upperarm_l [{shapeType=Capsule, length=13.005, radius=5.5, shapeName=upperarm_l_capsule, center=<10 0 -1>, rotation=<rot 90 0 0>}]
}
constraints [
 {
  bone1Name=neck_01
  bone2Name=spine_05
  swing1LimitDegrees=45
  swing1Motion=Free
  swing2LimitDegrees=45
  swing2Motion=Free
  twistLimitDegrees=45
  twistMotion=Free
 }
 {
  bone1Name=head
  bone2Name=neck_02
  swing1LimitDegrees=45
  swing1Motion=Free
  swing2LimitDegrees=45
  swing2Motion=Free
  twistLimitDegrees=45
  twistMotion=Free
 }
 {
  bone1Name=clavicle_l
  bone2Name=spine_05
  swing1LimitDegrees=45
  swing1Motion=Free
  swing2LimitDegrees=45
  swing2Motion=Free
  twistLimitDegrees=45
  twistMotion=Free
 }
 {
  bone1Name=dyn_hair_base
  bone2Name=head
  swing1LimitDegrees=45
  swing1Motion=Free
  swing2LimitDegrees=45
  swing2Motion=Free
  twistLimitDegrees=45
  twistMotion=Free
 }
 {bone1Name=dyn_hair_01, bone2Name=dyn_hair_base, swing1LimitDegrees=30, swing2LimitDegrees=15, twistLimitDegrees=5}
 {bone1Name=dyn_hair_02, bone2Name=dyn_hair_01, swing1LimitDegrees=30, swing2LimitDegrees=15, twistLimitDegrees=5}
 {bone1Name=dyn_hair_03, bone2Name=dyn_hair_02, swing1LimitDegrees=40, swing2LimitDegrees=20, twistLimitDegrees=5}
 {bone1Name=dyn_hair_04, bone2Name=dyn_hair_03, swing1LimitDegrees=40, swing2LimitDegrees=18, twistLimitDegrees=5}
 {bone1Name=dyn_hair_05, bone2Name=dyn_hair_04, swing1LimitDegrees=40, swing2LimitDegrees=10, twistLimitDegrees=5}
 {bone1Name=dyn_hair_06, bone2Name=dyn_hair_05, swing1LimitDegrees=30, swing2LimitDegrees=10, twistLimitDegrees=5}
 {bone1Name=dyn_hair_07, bone2Name=dyn_hair_06, swing1LimitDegrees=10, swing2LimitDegrees=10, twistLimitDegrees=5}
 {bone1Name=dyn_hair_08, bone2Name=dyn_hair_06, swing1LimitDegrees=10, swing2LimitDegrees=10, twistLimitDegrees=5}
 {bone1Name=dyn_hair_09, bone2Name=dyn_hair_06, swing1LimitDegrees=10, swing2LimitDegrees=10, twistLimitDegrees=5}
 {
  bone1Name=upperarm_l
  bone2Name=clavicle_l
  swing1LimitDegrees=45
  swing1Motion=Free
  swing2LimitDegrees=45
  swing2Motion=Free
  twistLimitDegrees=45
  twistMotion=Free
 }
 {
  bone1Name=neck_02
  bone2Name=neck_01
  swing1LimitDegrees=45
  swing1Motion=Free
  swing2LimitDegrees=45
  swing2Motion=Free
  twistLimitDegrees=45
  twistMotion=Free
 }
 {
  bone1Name=clavicle_r
  bone2Name=spine_05
  swing1LimitDegrees=45
  swing1Motion=Free
  swing2LimitDegrees=45
  swing2Motion=Free
  twistLimitDegrees=45
  twistMotion=Free
 }
]
