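# Animation Blueprint for Ruby's face accessory that drives her hair bones over a pose copied from the parent and body meshes.
# Head-angle alphas drive corrective bone offsets, and back bling selects between free hair and a variant with weighted multi-bone surface constraints.
# Two dynamics chains follow, RigidBody-only or RigidBody plus AnimDynamics, joined by an unwired bool blend.
# The event graph updates gravity, skydiving inertia scales and dynamics resets.
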AnimBlueprint /BRCosmetics/Characters/Player/Female/Medium/Bodies/F_MED_Ruby/Meshes/Parts/F_MED_Ruby_FaceAcc_AnimBP
target_skeleton=&/F_MED_Ruby_FaceAcc_Skeleton
graphs [
 AnimGraph {nodes [
   n1 = Use cached pose 'CACHE_WithBackPack'()
   n2 = Use cached pose 'Cache_FreeHair'()
   n3 = Get WearingAtBack() -> (WearingAtBack:bool)
   n4 = "Transform (Modify) Bone - Bone: dyn_hair_base"!(
    ComponentPose=Local To Component(LocalPose=Blend Poses by bool(BlendPose_0=@n1, BlendPose_1=@n2, bActiveValue=@n3) {blendType=Linear})
    bAlphaBoolEnabled:bool=True
    Translation="2,-1.5,0"
    Rotation="0,0,0"
    Scale="1,1,1"
   ) -> (Pose) {
    alphaInputType=Bool
    rotationMode=BMM_Additive
    rotationSpace=BCS_BoneSpace
    scaleMode=BMM_Additive
    scaleSpace=BCS_BoneSpace
    translationMode=BMM_Additive
    translationSpace=BCS_BoneSpace
    alphaBoolBlend {blendInTime=5, blendOption=Linear, blendOutTime=5}
    boneToModify {boneName=dyn_hair_base}
    translation=<2 -1.5>
   }
   n5 = "Transform (Modify) Bone - Bone: base_lshoulderhair"!(ComponentPose=@n4, bAlphaBoolEnabled:bool=True, Translation="-3.5,0,0", Rotation="20,-20,0", Scale="1,1,1") -> (Pose) {
    alphaInputType=Bool
    rotationMode=BMM_Additive
    rotationSpace=BCS_BoneSpace
    scaleMode=BMM_Additive
    scaleSpace=BCS_BoneSpace
    translationMode=BMM_Additive
    translationSpace=BCS_BoneSpace
    alphaBoolBlend {blendInTime=5, blendOption=Linear, blendOutTime=5}
    boneToModify {boneName=base_lshoulderhair}
    rotation {pitch=20, yaw=-20}
    translation {x=-3.5}
   }
   n6 = "Transform (Modify) Bone - Bone: dyn_hair_l_base"!(ComponentPose=@n5, bAlphaBoolEnabled:bool=True, Translation="0,0,0", Rotation="0,-20,45", Scale="1,1,1") -> (Pose) {
    alphaInputType=Bool
    rotationMode=BMM_Additive
    rotationSpace=BCS_BoneSpace
    scaleMode=BMM_Additive
    scaleSpace=BCS_BoneSpace
    translationMode=BMM_Additive
    translationSpace=BCS_BoneSpace
    alphaBoolBlend {blendInTime=5, blendOption=Linear, blendOutTime=5}
    boneToModify {boneName=dyn_hair_l_base}
    rotation {roll=45, yaw=-20}
   }
   n7 = "Constraint - dyn_hair_mid_04 (R)"!(ComponentPose=@n6, bAlphaBoolEnabled:bool=True, ConstraintWeights_0=1) -> (Pose) {
    alphaInputType=Bool
    alphaBoolBlend {blendOption=CubicInOut}
    boneToModify {boneName=dyn_hair_mid_04}
    constraintSetup [{transformType=Rotation, targetBone {boneName=dyn_hair_r_04}}]
    constraintWeights [1]
   }
   n8 = "Constraint - dyn_hair_mid_04 (R)"(ComponentPose=@n7, bAlphaBoolEnabled=True, ConstraintWeights_0=1) {
    alphaInputType=Bool
    alphaBoolBlend {blendOption=CubicInOut}
    boneToModify {boneName=dyn_hair_mid_04}
    constraintSetup [{transformType=Rotation, targetBone {boneName=dyn_hair_r_04}}]
    constraintWeights [1]
   }
   n9 = "Transform (Modify) Bone - Bone: dyn_hair_l_04"!(ComponentPose=@n8, bAlphaBoolEnabled:bool=True, Translation="0,0,0", Rotation="0,-20,0", Scale="1,1,1") -> (Pose) {
    alphaInputType=Bool
    rotationMode=BMM_Additive
    rotationSpace=BCS_BoneSpace
    scaleMode=BMM_Additive
    scaleSpace=BCS_BoneSpace
    translationMode=BMM_Additive
    translationSpace=BCS_BoneSpace
    alphaBoolBlend {blendOption=CubicInOut, blendOutTime=5}
    alphaScaleBiasClamp {bMapRange=true}
    boneToModify {boneName=dyn_hair_l_04}
    rotation {yaw=-20}
   }
   n10 = "Transform (Modify) Bone - Bone: dyn_hair_mid_04"!(ComponentPose=@n9, bAlphaBoolEnabled:bool=True, Translation="0,0,0", Rotation="0,-5,0", Scale="1,1,1") -> (Pose) {
    alphaInputType=Bool
    rotationMode=BMM_Additive
    rotationSpace=BCS_BoneSpace
    scaleMode=BMM_Additive
    scaleSpace=BCS_BoneSpace
    translationMode=BMM_Additive
    translationSpace=BCS_BoneSpace
    alphaBoolBlend {blendInTime=5, blendOption=Linear, blendOutTime=5}
    boneToModify {boneName=dyn_hair_mid_04}
    rotation {yaw=-5}
   }
   n11 = "Transform (Modify) Bone - Bone: dyn_hair_base"(ComponentPose=@n10, bAlphaBoolEnabled=True, Translation="0,0,0", Rotation="0,30,0", Scale="1,1,1") {
    alphaInputType=Bool
    rotationMode=BMM_Additive
    rotationSpace=BCS_BoneSpace
    scaleMode=BMM_Additive
    scaleSpace=BCS_BoneSpace
    translationMode=BMM_Additive
    translationSpace=BCS_BoneSpace
    alphaBoolBlend {blendInTime=5, blendOption=Linear, blendOutTime=5}
    boneToModify {boneName=dyn_hair_base}
    rotation {yaw=30}
   }
   n12 = "Constraint - dyn_hair_mid_04 (R)"(ComponentPose=@n11, bAlphaBoolEnabled=True, ConstraintWeights_0=1) {
    alphaInputType=Bool
    alphaBoolBlend {blendOption=CubicInOut}
    boneToModify {boneName=dyn_hair_mid_04}
    constraintSetup [{transformType=Rotation, targetBone {boneName=dyn_hair_r_04}}]
    constraintWeights [1]
   }
   n13 = Get Linear Scale Free() -> (Linear Scale Free:Vector)
   n14 = RigidBody!(
    ComponentPose=@n12
    Alpha
    OverrideWorldGravity:Vector=Get GravityOverride()
    ExternalForce
    ComponentLinearAccScale:Vector=@n13
    ComponentLinearVelScale:Vector=@n13
   ) -> (Pose) {
    bOverrideWorldGravity=true
    overridePhysicsAsset=&/F_MED_Ruby_FaceAcc_Back_Physics
    simulationSpace=BaseBoneSpace
    alphaBoolBlend {blendOption=Linear}
    baseBoneRef {boneName=neck_01}
    componentAppliedLinearAccClamp=<3000 3000 3000>
    componentLinearAccScale=<0.3 0.4 0.3>
    componentLinearVelScale=<0.3 0.4 0.3>
    simSpaceSettings {worldAlpha=0.05}
   }
   n15 = Get Linear Scale Free()
   n16 = RigidBody(ComponentPose=@n14, OverrideWorldGravity=Get GravityOverride(), ComponentLinearAccScale=@n15, ComponentLinearVelScale=@n15) {
    bClampLinearTranslationLimitToRefPose=true
    bOverrideWorldGravity=true
    overridePhysicsAsset=&/F_MED_Ruby_FaceAcc_Physics
    simulationSpace=BaseBoneSpace
    alphaBoolBlend {blendOption=Linear}
    baseBoneRef {boneName=neck_01}
    componentAppliedLinearAccClamp=<2000 2000 2000>
    componentLinearAccScale=<0.6 0.2 0.2>
    componentLinearVelScale=<0.6 0.2 0.2>
    simSpaceSettings {worldAlpha=0.05}
   }
   n17 = Use cached pose 'CACHE_WithBackPack'()
   n18 = Use cached pose 'Cache_FreeHair'()
   n19 = Get WearingAtBack()
   n20 = "Constraint - dyn_hair_02 (PPPPPP)"!(
    ComponentPose=Local To Component(LocalPose=Blend Poses by bool(BlendPose_0=@n17, BlendPose_1=@n18, bActiveValue=@n19) {blendType=Linear})
    Alpha=1
    ConstraintWeights_0=0.3796
    ConstraintWeights_1:float=0.3652
    ConstraintWeights_2:float=0.1859
    ConstraintWeights_3:float=0.0597
    ConstraintWeights_4:float=0.0058
    ConstraintWeights_5:float=0.0039
   ) -> (Pose) {
    alphaBoolBlend {blendOption=Linear}
    boneToModify {boneName=dyn_hair_02}
    constraintSetup [
     {targetBone {boneName=head}}
     {targetBone {boneName=neck_02}}
     {targetBone {boneName=neck_01}}
     {targetBone {boneName=spine_05}}
     {targetBone {boneName=spine_04}}
     {targetBone {boneName=clavicle_l}}
    ]
    constraintWeights [1,1,1,1,1,1]
   }
   n21 = "Constraint - dyn_hair_03 (PPPPPP)"!(
    ComponentPose=@n20
    Alpha=1
    ConstraintWeights_0=0.393
    ConstraintWeights_1:float=0.2737
    ConstraintWeights_2:float=0.211
    ConstraintWeights_3:float=0.1064
    ConstraintWeights_4:float=0.0105
    ConstraintWeights_5:float=0.0054
   ) -> (Pose) {
    alphaBoolBlend {blendOption=Linear}
    boneToModify {boneName=dyn_hair_03}
    constraintSetup [
     {targetBone {boneName=neck_02}}
     {targetBone {boneName=head}}
     {targetBone {boneName=neck_01}}
     {targetBone {boneName=spine_05}}
     {targetBone {boneName=spine_04}}
     {targetBone {boneName=clavicle_l}}
    ]
    constraintWeights [1,1,1,1,1,1]
   }
   n22 = "Transform (Modify) Bone - Bone: dyn_hair_03"!(ComponentPose=@n21, bAlphaBoolEnabled:bool=True, Translation="-0.35,0.25,0") -> (Pose) {
    alphaInputType=Bool
    rotationMode=BMM_Additive
    rotationSpace=BCS_BoneSpace
    translationMode=BMM_Additive
    translationSpace=BCS_BoneSpace
    alphaBoolBlend {blendOption=Linear}
    boneToModify {boneName=dyn_hair_03}
    rotation {yaw=7}
    translation=<-0.35 0.25>
   }
   n23 = RigidBody!(ComponentPose=@n22, Alpha, OverrideWorldGravity="0,0,0", ComponentLinearAccScale="0.3,0.4,0.3", ComponentLinearVelScale="0.3,0.4,0.3") -> (Pose) {
    bOverrideWorldGravity=true
    overridePhysicsAsset=&/F_MED_Ruby_FaceAcc_FrontEnd_Physics
    simulationSpace=BaseBoneSpace
    alphaBoolBlend {blendOption=Linear}
    baseBoneRef {boneName=spine_03}
    componentAppliedLinearAccClamp=<3000 3000 3000>
    componentLinearAccScale=<0.3 0.4 0.3>
    componentLinearVelScale=<0.3 0.4 0.3>
    simSpaceSettings {
     maxAngularAcceleration=1000
     maxAngularVelocity=3
     maxLinearAcceleration=700
     maxLinearVelocity=500
     velocityScaleZ=0
     worldAlpha=0.65
     externalLinearDragV=<0.2 0.2>
    }
   }
   n24 = "Transform (Modify) Bone - Bone: dyn_hair_02"!(ComponentPose=@n23, bAlphaBoolEnabled:bool=True, Translation="-0.25,0.25,0") -> (Pose) {
    alphaInputType=Bool
    rotationMode=BMM_Additive
    rotationSpace=BCS_BoneSpace
    translationMode=BMM_Additive
    translationSpace=BCS_BoneSpace
    alphaBoolBlend {blendOption=Linear}
    boneToModify {boneName=dyn_hair_02}
    translation=<-0.25 0.25>
   }
   n25 = "Transform (Modify) Bone - Bone: dyn_hair_06"!(ComponentPose=@n24, bAlphaBoolEnabled:bool=True, Rotation="0,7,0") -> (Pose) {
    alphaInputType=Bool
    rotationMode=BMM_Additive
    rotationSpace=BCS_BoneSpace
    translationMode=BMM_Additive
    translationSpace=BCS_BoneSpace
    alphaBoolBlend {blendOption=Linear}
    boneToModify {boneName=dyn_hair_06}
    rotation {yaw=7}
   }
   n26 = "Constraint - dyn_hair_l_03 (PPPPPPP)"!(
    ComponentPose=@n25
    Alpha=1
    ConstraintWeights_0=0.7519
    ConstraintWeights_1:float=0.1762
    ConstraintWeights_2:float=0.0523
    ConstraintWeights_3:float=0.0124
    ConstraintWeights_4:float=0.0037
    ConstraintWeights_5:float=0.002
    ConstraintWeights_6:float=0.0015
   ) -> (Pose) {
    alphaBoolBlend {blendOption=Linear}
    boneToModify {boneName=dyn_hair_l_03}
    constraintSetup [
     {targetBone {boneName=spine_05}}
     {targetBone {boneName=clavicle_l}}
     {targetBone {boneName=neck_01}}
     {targetBone {boneName=neck_02}}
     {targetBone {boneName=spine_04}}
     {targetBone {boneName=clavicle_r}}
     {targetBone {boneName=head}}
    ]
    constraintWeights [1,1,1,1,1,1,1]
   }
   n27 = "Constraint - dyn_hair_mid_03 (PPPPPPP)"!(
    ComponentPose=@n26
    Alpha=1
    ConstraintWeights_0=0.8688
    ConstraintWeights_1:float=0.0492
    ConstraintWeights_2:float=0.0334
    ConstraintWeights_3:float=0.0314
    ConstraintWeights_4:float=0.0097
    ConstraintWeights_5:float=0.0058
    ConstraintWeights_6:float=0.0017
   ) -> (Pose) {
    alphaBoolBlend {blendOption=Linear}
    boneToModify {boneName=dyn_hair_mid_03}
    constraintSetup [
     {targetBone {boneName=spine_05}}
     {targetBone {boneName=neck_01}}
     {targetBone {boneName=clavicle_r}}
     {targetBone {boneName=clavicle_l}}
     {targetBone {boneName=neck_02}}
     {targetBone {boneName=spine_04}}
     {targetBone {boneName=head}}
    ]
    constraintWeights [1,1,1,1,1,1,1]
   }
   n28 = "Constraint - dyn_hair_r_03 (PPPPPP)"!(
    ComponentPose=@n27
    Alpha=1
    ConstraintWeights_0=0.6862
    ConstraintWeights_1:float=0.2864
    ConstraintWeights_2:float=0.0192
    ConstraintWeights_3:float=0.005
    ConstraintWeights_4:float=0.0019
    ConstraintWeights_5:float=0.0013
   ) -> (Pose) {
    alphaBoolBlend {blendOption=Linear}
    boneToModify {boneName=dyn_hair_r_03}
    constraintSetup [
     {targetBone {boneName=spine_05}}
     {targetBone {boneName=clavicle_r}}
     {targetBone {boneName=neck_01}}
     {targetBone {boneName=neck_02}}
     {targetBone {boneName=upperarm_r}}
     {targetBone {boneName=upperarm_twist_01_r}}
    ]
    constraintWeights [1,1,1,1,1,1]
   }
   n29 = Anim Dynamics - Bone: dyn_hair_l_03!(ComponentPose=@n28, Alpha=0.8) -> (Pose) {
    angularDampingOverride=1
    angularSpringConstant=25000
    bAngularSpring=true
    bOverrideAngularDamping=true
    bUseGravityOverride=true
    gravityScale=0.35
    simulationSpace=World
    alphaBoolBlend {blendOption=Linear}
    boundBone {boneName=dyn_hair_l_03}
    componentLinearAccScale=<1 1 1>
    componentLinearVelScale=<1 1 1>
    gravityOverride {z=-980}
    physicsBodyDefinitions [{
      boundBone {boneName=dyn_hair_l_03}
      boxExtents=<25 35 35>
      constraintSetup {angularLimitsMax=<20 20 20>, angularLimitsMin=<-20 0 -20>, angularTarget=<1 0 0>}
      localJointOffset=<0.5 0 0>
     }]
   }
   n30 = Anim Dynamics - Bone: dyn_hair_mid_03!(ComponentPose=@n29, Alpha=0.8) -> (Pose) {
    angularDampingOverride=1
    angularSpringConstant=25000
    bAngularSpring=true
    bOverrideAngularDamping=true
    bUseGravityOverride=true
    gravityScale=0.35
    simulationSpace=World
    alphaBoolBlend {blendOption=Linear}
    boundBone {boneName=dyn_hair_mid_03}
    componentLinearAccScale=<1 1 1>
    componentLinearVelScale=<1 1 1>
    gravityOverride {z=-980}
    physicsBodyDefinitions [{
      boundBone {boneName=dyn_hair_mid_03}
      boxExtents=<25 35 35>
      constraintSetup {angularLimitsMax=<20 20 20>, angularLimitsMin=<-20 0 -20>, angularTarget=<1 0 0>}
      localJointOffset=<0.5 0 0>
     }]
   }
   n31 = Anim Dynamics - Bone: dyn_hair_r_03!(ComponentPose=@n30, Alpha=0.8) -> (Pose) {
    angularDampingOverride=1
    angularSpringConstant=25000
    bAngularSpring=true
    bOverrideAngularDamping=true
    bUseGravityOverride=true
    gravityScale=0.35
    simulationSpace=World
    alphaBoolBlend {blendOption=Linear}
    boundBone {boneName=dyn_hair_r_03}
    componentLinearAccScale=<1 1 1>
    componentLinearVelScale=<1 1 1>
    gravityOverride {z=-980}
    physicsBodyDefinitions [{
      boundBone {boneName=dyn_hair_r_03}
      boxExtents=<25 35 35>
      constraintSetup {angularLimitsMax=<20 20 20>, angularLimitsMin=<-20 0 -20>, angularTarget=<1 0 0>}
      localJointOffset=<0.5 0 0>
     }]
   }
   n32 = Anim Dynamics - Bone: dyn_hair_mid_04!(ComponentPose=@n31, Alpha=0.8) -> (Pose) {
    angularDampingOverride=1
    angularSpringConstant=25000
    bAngularSpring=true
    bOverrideAngularDamping=true
    bUseGravityOverride=true
    gravityScale=0.35
    simulationSpace=World
    alphaBoolBlend {blendOption=Linear}
    boundBone {boneName=dyn_hair_mid_04}
    componentLinearAccScale=<1 1 1>
    componentLinearVelScale=<1 1 1>
    gravityOverride {z=-980}
    physicsBodyDefinitions [{
      boundBone {boneName=dyn_hair_mid_04}
      boxExtents=<35 45 45>
      constraintSetup {angularLimitsMax=<10 10 10>, angularLimitsMin=<-10 0 -10>, angularTarget=<1 0 0>}
      localJointOffset=<0.5 0 0>
     }]
   }
   n33 = Anim Dynamics - Bone: dyn_hair_r_04!(ComponentPose=@n32, Alpha=0.8) -> (Pose) {
    angularDampingOverride=1
    angularSpringConstant=25000
    bAngularSpring=true
    bOverrideAngularDamping=true
    bUseGravityOverride=true
    gravityScale=0.35
    simulationSpace=World
    alphaBoolBlend {blendOption=Linear}
    boundBone {boneName=dyn_hair_r_04}
    componentLinearAccScale=<1 1 1>
    componentLinearVelScale=<1 1 1>
    gravityOverride {z=-980}
    physicsBodyDefinitions [{
      boundBone {boneName=dyn_hair_r_04}
      boxExtents=<35 45 45>
      constraintSetup {angularLimitsMax=<10 10 10>, angularLimitsMin=<-10 0 -10>, angularTarget=<1 0 0>}
      localJointOffset=<0.5 0 0>
     }]
   }
   n34 = Anim Dynamics - Bone: dyn_hair_l_04!(ComponentPose=@n33, Alpha=0.8) -> (Pose) {
    angularDampingOverride=1
    angularSpringConstant=25000
    bAngularSpring=true
    bOverrideAngularDamping=true
    bUseGravityOverride=true
    gravityScale=0.35
    simulationSpace=World
    alphaBoolBlend {blendOption=Linear}
    boundBone {boneName=dyn_hair_l_04}
    componentLinearAccScale=<1 1 1>
    componentLinearVelScale=<1 1 1>
    gravityOverride {z=-980}
    physicsBodyDefinitions [{
      boundBone {boneName=dyn_hair_l_04}
      boxExtents=<35 45 45>
      constraintSetup {angularLimitsMax=<10 10 10>, angularLimitsMin=<-10 0 -10>, angularTarget=<1 0 0>}
      localJointOffset=<0.5 0 0>
     }]
   }
   Output Pose(Result=Blend Poses by bool(BlendPose_0=Component To Local(ComponentPose=@n34), BlendPose_1=Component To Local(ComponentPose=@n16)) {blendType=Linear}) {layerGroup=None}
   n35 = "Transform (Modify) Bone - Bone: dyn_hair_r_base"(
    ComponentPose=Local To Component(LocalPose=Use cached pose 'Cache_SurfaceConstraints'())
    Alpha=Get Head Up() -> (Head Up:double)
    Translation="-5,1,-1"
    Rotation="-10,5,-10"
    Scale="1,1,1"
   ) {
    rotationMode=BMM_Additive
    rotationSpace=BCS_BoneSpace
    scaleMode=BMM_Additive
    scaleSpace=BCS_BoneSpace
    translationMode=BMM_Additive
    translationSpace=BCS_BoneSpace
    alphaBoolBlend {blendOption=Linear}
    boneToModify {boneName=dyn_hair_r_base}
    rotation=<rot -10 5 -10>
    translation=<-5 1 -1>
   }
   n36 = "Transform (Modify) Bone - Bone: dyn_hair_r_base"(ComponentPose=@n35, Alpha=Get Head Down() -> (Head Down:double), Translation="-11,0,0", Rotation="0,0,0", Scale="1,1,1") {
    rotationMode=BMM_Additive
    rotationSpace=BCS_BoneSpace
    scaleMode=BMM_Additive
    scaleSpace=BCS_BoneSpace
    translationMode=BMM_Additive
    translationSpace=BCS_BoneSpace
    alphaBoolBlend {blendOption=Linear}
    boneToModify {boneName=dyn_hair_r_base}
    translation {x=-11}
   }
   n37 = "Transform (Modify) Bone - Bone: dyn_hair_r_base"(ComponentPose=@n36, Alpha=Get Head Rot L() -> (Head Rot L:double), Translation="-1,0,-1.5", Rotation="0,0,0", Scale="1,1,1") {
    rotationMode=BMM_Additive
    rotationSpace=BCS_BoneSpace
    scaleMode=BMM_Additive
    scaleSpace=BCS_BoneSpace
    translationMode=BMM_Additive
    translationSpace=BCS_BoneSpace
    alphaBoolBlend {blendOption=Linear}
    boneToModify {boneName=dyn_hair_r_base}
    translation {x=-1, z=-1.5}
   }
   n38 = "Transform (Modify) Bone - Bone: dyn_hair_r_base"!(ComponentPose=@n37, bAlphaBoolEnabled:bool=True, Translation="-0.5,0,0", Rotation="0,0,0", Scale="1,1,1") -> (Pose) {
    alphaInputType=Bool
    rotationMode=BMM_Additive
    rotationSpace=BCS_BoneSpace
    scaleMode=BMM_Additive
    scaleSpace=BCS_BoneSpace
    translationMode=BMM_Additive
    translationSpace=BCS_BoneSpace
    alphaBoolBlend {blendOption=CubicInOut, blendOutTime=5}
    alphaScaleBiasClamp {bMapRange=true}
    boneToModify {boneName=dyn_hair_r_base}
    translation {x=-0.5}
   }
   Save cached pose 'CACHE_WithBackPack'(Pose=Component To Local(ComponentPose=@n38)) {cacheName=CACHE_WithBackPack}
   n39 = "Constraint - dyn_hair_l_02 (PPPPPP)"!(
    ComponentPose=Local To Component(LocalPose=Use cached pose 'Cache_StartPose'())
    Alpha=1
    ConstraintWeights_0=0.348
    ConstraintWeights_1:float=0.3339
    ConstraintWeights_2:float=0.2589
    ConstraintWeights_3:float=0.0474
    ConstraintWeights_4:float=0.0087
    ConstraintWeights_5:float=0.0032
   ) -> (Pose) {
    alphaBoolBlend {blendOption=Linear}
    boneToModify {boneName=dyn_hair_l_02}
    constraintSetup [
     {targetBone {boneName=head}}
     {targetBone {boneName=neck_02}}
     {targetBone {boneName=neck_01}}
     {targetBone {boneName=spine_05}}
     {targetBone {boneName=spine_04}}
     {targetBone {boneName=clavicle_l}}
    ]
    constraintWeights [1,1,1,1,1,1]
   }
   n40 = "Constraint - dyn_hair_mid_02 (PPPPPPP)"!(
    ComponentPose=@n39
    Alpha=1
    ConstraintWeights_0=0.3246
    ConstraintWeights_1:float=0.317
    ConstraintWeights_2:float=0.3014
    ConstraintWeights_3:float=0.046
    ConstraintWeights_4:float=0.0084
    ConstraintWeights_5:float=0.0013
    ConstraintWeights_6:float=0.0013
   ) -> (Pose) {
    alphaBoolBlend {blendOption=Linear}
    boneToModify {boneName=dyn_hair_mid_02}
    constraintSetup [
     {targetBone {boneName=neck_02}}
     {targetBone {boneName=head}}
     {targetBone {boneName=neck_01}}
     {targetBone {boneName=spine_05}}
     {targetBone {boneName=spine_04}}
     {targetBone {boneName=clavicle_r}}
     {targetBone {boneName=clavicle_l}}
    ]
    constraintWeights [1,1,1,1,1,1,1]
   }
   n41 = "Constraint - dyn_hair_r_02 (PPPPPP)"!(
    ComponentPose=@n40
    Alpha=1
    ConstraintWeights_0=0.3518
    ConstraintWeights_1:float=0.3379
    ConstraintWeights_2:float=0.2496
    ConstraintWeights_3:float=0.0489
    ConstraintWeights_4:float=0.0082
    ConstraintWeights_5:float=0.0036
   ) -> (Pose) {
    alphaBoolBlend {blendOption=Linear}
    boneToModify {boneName=dyn_hair_r_02}
    constraintSetup [
     {targetBone {boneName=head}}
     {targetBone {boneName=neck_02}}
     {targetBone {boneName=neck_01}}
     {targetBone {boneName=spine_05}}
     {targetBone {boneName=spine_04}}
     {targetBone {boneName=clavicle_r}}
    ]
    constraintWeights [1,1,1,1,1,1]
   }
   n42 = "Constraint - dyn_hair_l_03 (PPPPPP)"!(
    ComponentPose=@n41
    Alpha=1
    ConstraintWeights_0=0.3285
    ConstraintWeights_1:float=0.3158
    ConstraintWeights_2:float=0.1732
    ConstraintWeights_3:float=0.1523
    ConstraintWeights_4:float=0.018
    ConstraintWeights_5:float=0.0123
   ) -> (Pose) {
    alphaBoolBlend {blendOption=Linear}
    boneToModify {boneName=dyn_hair_l_03}
    constraintSetup [
     {targetBone {boneName=neck_01}}
     {targetBone {boneName=neck_02}}
     {targetBone {boneName=head}}
     {targetBone {boneName=spine_05}}
     {targetBone {boneName=spine_04}}
     {targetBone {boneName=clavicle_l}}
    ]
    constraintWeights [1,1,1,1,1,1]
   }
   n43 = "Constraint - dyn_hair_mid_03 (PPPPPPP)"(
    ComponentPose=@n42
    Alpha=1
    ConstraintWeights_0=0.3518
    ConstraintWeights_1=0.2555
    ConstraintWeights_2=0.2433
    ConstraintWeights_3=0.1098
    ConstraintWeights_4=0.024
    ConstraintWeights_5=0.0078
    ConstraintWeights_6=0.0078
   ) {
    alphaBoolBlend {blendOption=Linear}
    boneToModify {boneName=dyn_hair_mid_03}
    constraintSetup [
     {targetBone {boneName=neck_01}}
     {targetBone {boneName=neck_02}}
     {targetBone {boneName=spine_05}}
     {targetBone {boneName=head}}
     {targetBone {boneName=spine_04}}
     {targetBone {boneName=clavicle_l}}
     {targetBone {boneName=clavicle_r}}
    ]
    constraintWeights [1,1,1,1,1,1,1]
   }
   n44 = "Constraint - dyn_hair_r_03 (PPPPPP)"(
    ComponentPose=@n43
    Alpha=1
    ConstraintWeights_0=0.3209
    ConstraintWeights_1=0.3136
    ConstraintWeights_2=0.1679
    ConstraintWeights_3=0.1651
    ConstraintWeights_4=0.0179
    ConstraintWeights_5=0.0146
   ) {
    alphaBoolBlend {blendOption=Linear}
    boneToModify {boneName=dyn_hair_r_03}
    constraintSetup [
     {targetBone {boneName=neck_01}}
     {targetBone {boneName=neck_02}}
     {targetBone {boneName=spine_05}}
     {targetBone {boneName=head}}
     {targetBone {boneName=spine_04}}
     {targetBone {boneName=clavicle_r}}
    ]
    constraintWeights [1,1,1,1,1,1]
   }
   n45 = "Constraint - dyn_hair_l_04 (PPPPPPP)"!(
    ComponentPose=@n44
    Alpha=1
    ConstraintWeights_0=0.7519
    ConstraintWeights_1:float=0.1762
    ConstraintWeights_2:float=0.0523
    ConstraintWeights_3:float=0.0124
    ConstraintWeights_4:float=0.0037
    ConstraintWeights_5:float=0.002
    ConstraintWeights_6:float=0.0015
   ) -> (Pose) {
    alphaBoolBlend {blendOption=Linear}
    boneToModify {boneName=dyn_hair_l_04}
    constraintSetup [
     {targetBone {boneName=spine_05}}
     {targetBone {boneName=clavicle_l}}
     {targetBone {boneName=neck_01}}
     {targetBone {boneName=neck_02}}
     {targetBone {boneName=spine_04}}
     {targetBone {boneName=clavicle_r}}
     {targetBone {boneName=head}}
    ]
    constraintWeights [1,1,1,1,1,1,1]
   }
   n46 = "Constraint - dyn_hair_mid_04 (PPPPPPP)"!(
    ComponentPose=@n45
    Alpha=1
    ConstraintWeights_0=0.8688
    ConstraintWeights_1:float=0.0492
    ConstraintWeights_2:float=0.0334
    ConstraintWeights_3:float=0.0314
    ConstraintWeights_4:float=0.0097
    ConstraintWeights_5:float=0.0058
    ConstraintWeights_6:float=0.0017
   ) -> (Pose) {
    alphaBoolBlend {blendOption=Linear}
    boneToModify {boneName=dyn_hair_mid_04}
    constraintSetup [
     {targetBone {boneName=spine_05}}
     {targetBone {boneName=neck_01}}
     {targetBone {boneName=clavicle_r}}
     {targetBone {boneName=clavicle_l}}
     {targetBone {boneName=neck_02}}
     {targetBone {boneName=spine_04}}
     {targetBone {boneName=head}}
    ]
    constraintWeights [1,1,1,1,1,1,1]
   }
   n47 = "Constraint - dyn_hair_r_04 (PPPPTT)"!(
    ComponentPose=@n46
    Alpha=1
    ConstraintWeights_0=0.6862
    ConstraintWeights_1:float=0.2864
    ConstraintWeights_2:float=0.0192
    ConstraintWeights_3:float=0.005
    ConstraintWeights_4:float=0
    ConstraintWeights_5:float=0
   ) -> (Pose) {
    alphaBoolBlend {blendOption=Linear}
    boneToModify {boneName=dyn_hair_r_04}
    constraintSetup [
     {targetBone {boneName=spine_05}}
     {targetBone {boneName=clavicle_r}}
     {targetBone {boneName=neck_01}}
     {targetBone {boneName=neck_02}}
     {transformType=Translation, targetBone {boneName=None}}
     {transformType=Translation, targetBone {boneName=None}}
    ]
    constraintWeights [1,1,1,1,1,1]
   }
   Save cached pose 'Cache_SurfaceConstraints'(Pose=Component To Local(ComponentPose=@n47)) {cacheName=Cache_SurfaceConstraints}
   n48 = Get Head Up()
   n49 = "Transform (Modify) Bone - Bone: base_lshoulderhair"!(
    ComponentPose=Local To Component(LocalPose=Use cached pose 'Cache_StartPose'())
    Alpha:float=@n48
    Translation="-1,-1,-0.5"
    Rotation="0,0,0"
    Scale="1,1,1"
   ) -> (Pose) {
    rotationMode=BMM_Additive
    rotationSpace=BCS_BoneSpace
    scaleMode=BMM_Replace
    scaleSpace=BCS_BoneSpace
    translationMode=BMM_Additive
    translationSpace=BCS_BoneSpace
    alphaBoolBlend {blendOption=Linear}
    boneToModify {boneName=base_lshoulderhair}
    translation=<-1 -1 -0.5>
   }
   n50 = "Transform (Modify) Bone - Bone: base_innerhair"(ComponentPose=@n49, Alpha=@n48, Translation="3,2,0", Rotation="0,0,0", Scale="1,1,1") {
    rotationMode=BMM_Additive
    rotationSpace=BCS_BoneSpace
    scaleMode=BMM_Replace
    scaleSpace=BCS_BoneSpace
    translationMode=BMM_Additive
    translationSpace=BCS_BoneSpace
    alphaBoolBlend {blendOption=Linear}
    boneToModify {boneName=base_innerhair}
    translation=<3 2>
   }
   n51 = "Transform (Modify) Bone - Bone: dyn_hair_mid_base"(ComponentPose=@n50, Alpha=@n48, Translation="0.6,-0.5,0", Rotation="-7,-60,35", Scale="1,1,1") {
    rotationMode=BMM_Additive
    rotationSpace=BCS_BoneSpace
    scaleMode=BMM_Replace
    scaleSpace=BCS_BoneSpace
    translationMode=BMM_Additive
    translationSpace=BCS_BoneSpace
    alphaBoolBlend {blendOption=Linear}
    boneToModify {boneName=dyn_hair_mid_base}
    rotation=<rot -7 -60 35>
    translation=<0.6 -0.5>
   }
   n52 = "Transform (Modify) Bone - Bone: dyn_hair_r_base"!(ComponentPose=@n51, Alpha:float=@n48, Translation="-2,0.5,0", Rotation="-22,-50,35", Scale="1,1,1") -> (Pose) {
    rotationMode=BMM_Additive
    rotationSpace=BCS_BoneSpace
    scaleMode=BMM_Additive
    scaleSpace=BCS_BoneSpace
    translationMode=BMM_Additive
    translationSpace=BCS_BoneSpace
    alphaBoolBlend {blendOption=Linear}
    boneToModify {boneName=dyn_hair_r_base}
    rotation=<rot -22 -50 35>
    translation=<-2 0.5>
   }
   n53 = "Transform (Modify) Bone - Bone: base_lshoulderhair"(ComponentPose=@n52, Alpha=Get Head Rot L(), Translation="-0.5,0,0", Rotation="0,15,0", Scale="1,1,1") {
    rotationMode=BMM_Additive
    rotationSpace=BCS_BoneSpace
    scaleMode=BMM_Replace
    scaleSpace=BCS_BoneSpace
    translationMode=BMM_Additive
    translationSpace=BCS_BoneSpace
    alphaBoolBlend {blendOption=Linear}
    boneToModify {boneName=base_lshoulderhair}
    rotation {yaw=15}
    translation {x=-0.5}
   }
   n54 = "Transform (Modify) Bone - Bone: base_lshoulderhair"(ComponentPose=@n53, Alpha=Get Head Rot R() -> (Head Rot R:double), Translation="-0.5,0,0", Rotation="0,0,0", Scale="1,1,1") {
    rotationMode=BMM_Additive
    rotationSpace=BCS_BoneSpace
    scaleMode=BMM_Replace
    scaleSpace=BCS_BoneSpace
    translationMode=BMM_Additive
    translationSpace=BCS_BoneSpace
    alphaBoolBlend {blendOption=Linear}
    boneToModify {boneName=base_lshoulderhair}
    translation {x=-0.5}
   }
   n55 = "Transform (Modify) Bone - Bone: dyn_hair_base"!(ComponentPose=@n54, Alpha:float=Get Head Down(), Translation="-11,11,0", Rotation="0,-50,0", Scale="1,1,1") -> (Pose) {
    rotationMode=BMM_Additive
    rotationSpace=BCS_BoneSpace
    scaleMode=BMM_Replace
    scaleSpace=BCS_BoneSpace
    translationMode=BMM_Additive
    translationSpace=BCS_BoneSpace
    alphaBoolBlend {blendOption=Linear}
    boneToModify {boneName=dyn_hair_base}
    rotation {yaw=-50}
    translation=<-11 11>
   }
   Save cached pose 'Cache_FreeHair'(Pose=Component To Local(ComponentPose=@n55)) {cacheName=Cache_FreeHair}
   n56 = Layered blend per bone(
    BasePose=Copy Pose From Mesh() {bCopyCurves=true, bUseAttachedParent=true}
    BlendPoses_0=Copy Pose From Mesh(SourceMeshComponent=Get Body_Mesh())
   ) {layerSetup [{branchFilters [
       {boneName=clavicle_l}
       {boneName=clavicle_r}
       {boneName=upperarm_l}
       {boneName=upperarm_r}
       {boneName=upperarm_twist_01_l}
       {boneName=upperarm_twist_01_r}
       {boneName=deform_upperarm_bk_base_l}
       {boneName=deform_upperarm_bk_l}
       {boneName=deform_upperarm_fr_base_l}
       {boneName=deform_upperarm_fr_l}
       {boneName=deform_upperarm_up_base_l}
       {boneName=deform_upperarm_up_l}
       {boneName=deform_upperarm_bk_base_r}
       {boneName=deform_upperarm_bk_r}
       {boneName=deform_upperarm_fr_base_r}
       {boneName=deform_upperarm_fr_r}
       {boneName=deform_upperarm_up_base_r}
       {boneName=deform_upperarm_up_r}
      ]}]}
   Save cached pose 'Cache_StartPose'(Pose=@n56) {cacheName=Cache_StartPose}
  ]}
 EventGraph {nodes [
   Event BlueprintUpdateAnimation()
   > n1 = Sequence!(execute) -> (then_0, then_1, then_2:Exec, then_3:Exec, then_4:Exec, then_5:Exec)
   n2 = Get bIsFalling()
   n3 = Get bIsPlayingMeleeAnim()
   n4 = Get bIsTacticalSprinting()
   n5 = Branch(
    execute=@n1.then_2
    Condition=OR Boolean!(A=Get bIsParachuteOpened(), B=Get bIsJumping(), C:bool=AND Boolean(A=@n4, B=@n2), D:bool=@n2, E:bool=AND Boolean(A=@n2, B=@n3)) -> (ReturnValue)
   )
   Do Once(execute=@n5.else, Reset=@n5.then, Start Closed=true)
   > ResetDynamics(InTeleportType=TeleportPhysics)
   > ResetDynamics(InTeleportType=ResetPhysics)
   n6 = Branch(execute=@n1.then_1, Condition=Get bIsSkyDiving())
   > Set Linear Scale Free(Linear Scale Free:Vector="0.35,0.25,0.25") -> (Output_Get:Vector)
   > Set Linear Scale Backpack(Linear Scale Backpack:Vector="0.3,0.2,0.2") -> (Output_Get:Vector)
   > Set Linear Scale Cape(Linear Scale Cape:Vector="0.25,0.2,0.2") -> (Output_Get:Vector)
   Set Linear Scale Free(execute=@n6.else, Linear Scale Free="0.32,0.35,0.32")
   > Set Linear Scale Backpack(Linear Scale Backpack="0.2,0.2,0.2")
   > Set Linear Scale Cape(Linear Scale Cape="0.25,0.2,0.2")
   n7 = Gravity Override State Function DEPRECATED(
    execute=@n1.then_0
    CharacterPartAnimInstance=Self-Reference()
    GameTimeInSeconds=GetGameTimeInSeconds()
    GravityOverrideParams=Get Gravity Override Params()
   )
   > Set GravityOverride(GravityOverride=@n7)
   n8 = CheckForBackbling()
   > Set WearingAtBack(WearingAtBack:bool=OR Boolean!(A=@n8, B=@n8.bIsBackblingACape, C=@n8.bIsBackblingLong) -> (ReturnValue)) -> (Output_Get:bool)
   > Set isWearingBackpack(isWearingBackpack:bool=AND Boolean!(A=@n8, B=NOT Boolean(A=@n8.bIsBackblingACape), C:bool=NOT Boolean(A=@n8.bIsBackblingLong)) -> (ReturnValue)) -> (Output_Get:bool)
   > Set Is Wearing Cape(Is Wearing Cape=OR Boolean!(A=@n8.bIsBackblingACape, B=@n8.bIsBackblingLong) -> (ReturnValue))
   n9 = Get Skel Mesh()
   Set Head Up(
    execute=@n1.then_3
    Head Up:double=GetAlphaFromReferenceTransformDelta(SkeletalMeshComponent=@n9, BoneName=spine_05, BaseName=head, InputRange=(X=0,Y=70), TwistAxis="0,0,1")
   ) -> (Output_Get:double)
   > Set Head Down(Head Down:double=GetAlphaFromReferenceTransformDelta(SkeletalMeshComponent=@n9, BoneName=head, BaseName=neck_02, InputRange=(X=0,Y=90), TwistAxis="0,0,1")) -> (Output_Get:double)
   > Set Head Rot L(Head Rot L:double=GetAlphaFromReferenceTransformDelta(SkeletalMeshComponent=@n9, BoneName=spine_05, BaseName=head, DeltaComponentType=EulerX, InputRange=(X=0,Y=-22), TwistAxis="1,0,0")) -> (Output_Get:double)
   > Set Head Rot R(Head Rot R:double=GetAlphaFromReferenceTransformDelta(SkeletalMeshComponent=@n9, BoneName=spine_05, BaseName=head, DeltaComponentType=EulerX, InputRange=(X=0,Y=22), TwistAxis="1,0,0")) -> (Output_Get:double)
   Event OnCharacterCustomizationLoaded()
   > n10 = Cast To PlayerPawn_Athena(Object=TryGetPawnOwner())
   > Set Body_Mesh(Body_Mesh=GetSkeletalMeshForPartType(self=@n10))
   Event BlueprintInitializeAnimation()
   > Set Skel Mesh(Skel Mesh=GetOwnerSkeletalMeshComponent())
   Delay(Duration=0.2)
  ]}
]
variables [
 Body_Mesh {type=SkeletalMeshComponent}
 Gravity Override Params {type=GravityOverrideParamsStruct}
 GravityOverride {type=Vector}
 Head Down {type=number}
 Head Rot L {type=number}
 Head Rot R {type=number}
 Head Up {type=number}
 Is Wearing Cape {type=boolean}
 isWearingBackpack {type=boolean}
 JumpAndMelee {type=number}
 Linear Scale Backpack {type=Vector}
 Linear Scale Cape {type=Vector}
 Linear Scale Free {type=Vector}
 Skel Mesh {type=SkeletalMeshComponent}
 WearingAtBack {type=boolean}
]
